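# Material function layering hologram and glitch effects onto the IgniteEgg character skin, overriding incoming emissive, opacity mask and world position offset.
# Blueprint-fed timestamps drive a timed hologram transition, capped when DBNO, and expanding sphere-mask hit waves, with static-switch debug previews.
# Masks combine packed wireframe/random texture channels, panning local-position noise pulses, scanlines and object-stabilized screen-space digital noise.
# These feed fresnel-tinted glow, vertex glitch displacement and temporally dithered opacity.
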MaterialFunction /BRCosmetics/Characters/Player/Male/Medium/Bodies/M_MED_IgniteEgg/Prototype/Materials/MFC_IgniteEgg_FX
graph {nodes [
  Constant3Vector() {constant=<rgba 0.053 0.2591 3 1>}
  Frac(None=Multiply(A=Time(), B=ScalarParameter() {defaultValue=-0.5, group=IgniteEgg, parameterName=Debug_HologramSpeed}))
  n1 = GetMaterialAttributes(None=FunctionInput() {inputName=Material, inputType=FunctionInput_MaterialAttributes}) {attributeGetTypes [#1, #2, #3]}
  n2 = SetMaterialAttributes!(
   MaterialAttributes=@n1.MaterialAttributes
   OpacityMask=NamedRerouteUsage()
   EmissiveColor=NamedRerouteUsage()
   WorldPositionOffset=NamedRerouteUsage()
   PixelDepthOffset
  ) {attributeSetTypes [
    #1
    #2
    #3
    #4
   ]}
  FunctionOutput(None=@n2)
  n3 = VectorParameter() {group=IgniteEgg, parameterName=IE_HologramColor, defaultValue=<rgba 1.3877 1.0104 2 1>}
  n4 = LinearInterpolate(
   A=LinearInterpolate(A=@n3.RGB, B=Multiply(A=@n3.RGB) {constB=2}, Alpha=Fresnel() {baseReflectFraction=0, exponent=1})
   B=Multiply(A=@n3.RGB) {constB=3}
   Alpha=NamedRerouteUsage()
  )
  n5 = NamedRerouteUsage()
  n6 = NamedRerouteUsage()
  n7 = Max(A=NamedRerouteUsage(), B=@n6)
  n8 = NamedRerouteUsage()
  n9 = Multiply(A=@n8) {constB=0.6}
  n10 = LinearInterpolate(
   A=@n4
   B=Multiply(A=Multiply(A=@n5) {constB=2}, B=@n4)
   Alpha=Multiply(A=ScalarParameter() {defaultValue=0.5, parameterName=IE_GlitchColorIntensity}, B=Max(A=@n7, B=@n9))
  ) {constAlpha=0.1}
  n11 = NamedRerouteUsage()
  NamedRerouteDeclaration(None=LinearInterpolate(A=NamedRerouteUsage(), B=Multiply(A=@n10, B=@n11), Alpha=Multiply(A=@n8) {constB=0.5})) {name=Emissive_OUT, nodeColor=<rgba 0.8047 0 1 1>}
  n12 = VectorParameter() {group=IgniteEgg, parameterName=HologramTextureTilingXY, defaultValue=<rgba 7 7 0 1>}
  n13 = TextureSample(UVs=Multiply(A=TextureCoordinate(), B=ComponentMask(None=@n12.RGB) {g=true, r=true}) {constB=8}) {
   samplerSource=SSM_Wrap_WorldGroupSettings
   samplerType=SAMPLERTYPE_LinearColor
   texture=/BRCosmetics/Characters/Player/Male/Medium/Bodies/M_MED_IgniteEgg/Prototype/Textures/T_IgniteEgg_TA
  }
  NamedRerouteDeclaration(None=@n13.G) {name=Wireframe, nodeColor=<rgba 0 0.6719 1 1>}
  NamedRerouteDeclaration(None=@n1.WorldPositionOffset) {name=WPO_IN, nodeColor=<rgba 0.4297 1 0 1>}
  NamedRerouteDeclaration(None=@n6) {name=Debug, nodeColor=<rgba 1 0.6797 0 1>}
  n14 = ScalarParameter() {defaultValue=150, group=IgniteEgg, parameterName=IE_Scanline2_Amount}
  NamedRerouteDeclaration(None=LinearInterpolate(
    B=MaterialFunctionCall!(Scanline Amount=@n14) {materialFunction=/Game/Environments/Asteria/Sets/NeonCity/Techart/Holograms/NewHologramMaterial/MFScanlines}.Result
    Alpha=ScalarParameter() {defaultValue=1, group=IgniteEgg, parameterName=IE_Scanline2_Intensity}
   ) {constA=1}) {name=Scanlines, nodeColor=<rgba 0.4062 0 1 1>}
  n15 = MaterialFunctionCall!(Position, UVRadialScale) {materialFunction=/Game/Weapons/WeaponSkins/MaterialFunction/ObjectStabilizedScreenSpaceUVs}
  n16 = TextureSample(UVs=Panner(Coordinate=NamedRerouteDeclaration(None=@n15.StabilizedUVs) {name=ScreenSpaceUV, nodeColor=<rgba 0.4062 0 1 1>}) {bFractionalPart=true, speedX=-1.5, speedY=0.987}) {
   samplerSource=SSM_Wrap_WorldGroupSettings
   samplerType=SAMPLERTYPE_LinearColor
   texture=/Game/Global/Textures/Default/Noise/T_Atlas_DigitalNoise_01
  }
  n17 = ComponentMask(None=NamedRerouteDeclaration(None=@n16.RGB) {name=HologramGlitchColor, nodeColor=<rgba 1 0.1641 0 1>}) {g=true, r=true}
  n18 = MaterialFunctionCall(Float2=Clamp(None=ConstantBiasScale(None=@n17) {bias=-0.5, scale=2}) {minDefault=-1}) {materialFunction=&BreakOutFloat2Components}
  n19 = Constant()
  n20 = Multiply(
   A=Multiply(
    A=MaterialFunctionCall(X=@n18.R, Y=@n18.R, Z=@n19) {materialFunction=&MakeFloat3}.Result
    B=ScalarParameter() {defaultValue=0.3, group=IgniteEgg, parameterName=IE_GlitchWPOIntensity}
   )
   B=NamedRerouteUsage()
  )
  NamedRerouteDeclaration(None=Add(A=NamedRerouteUsage(), B=Transform(None=@n20) {transformSourceType=TRANSFORMSOURCE_Local})) {name=WPO_OUT, nodeColor=<rgba 0.2891 0 1 1>}
  n21 = ComponentMask(None=NamedRerouteUsage()) {b=true, r=true}
  n22 = Divide(A=@n21) {constB=280, desc=360}
  n23 = ScalarParameter() {group=IgniteEgg, parameterName=IE_RandomX}
  n24 = ScalarParameter() {group=IgniteEgg, parameterName=IE_RandomY}
  n25 = Multiply(A=MaterialFunctionCall(X=@n23, Y=@n24) {materialFunction=&MakeFloat2}.Result) {constB=8.3242}
  n26 = TextureSample(UVs=Panner(Coordinate=Add(A=@n22, B=@n25) {constB=0}) {speedX=0.05, speedY=0.082}) {
   samplerSource=SSM_Wrap_WorldGroupSettings
   samplerType=SAMPLERTYPE_LinearColor
   texture=&T_IgniteEgg_TA
  }
  n27 = Multiply(A=@n23, B=@n24)
  n28 = NamedRerouteUsage()
  n29 = LinearInterpolate(Alpha=@n28) {constA=1, constB=2, desc=Global frequency}
  n30 = Multiply(A=Add(A=Time(), B=Multiply(A=@n27) {constB=5.324}), B=Multiply(A=Constant() {desc=0.19453, r=0.1021}, B=@n29))
  n31 = Multiply(A=@n26.R, B=Power(Base=Sine(None=@n30), Exp=LinearInterpolate(Alpha=@n28) {constA=9, desc=Global frequency}) {constExponent=7.5})
  n32 = LinearInterpolate(Alpha=@n28) {constA=4.7, constB=2}
  n33 = MaterialFunctionCall(Speed=Multiply(A=Constant() {desc=0.34324, r=0.2}, B=@n29)) {materialFunction=&DebugTimeSine}
  n34 = Divide(A=@n21) {constB=128, desc=256}
  n35 = TextureSample(UVs=Panner(Coordinate=Add(A=@n25, B=@n34)) {speedX=-0.1, speedY=-0.3}) {
   samplerSource=SSM_Wrap_WorldGroupSettings
   samplerType=SAMPLERTYPE_LinearColor
   texture=&T_IgniteEgg_TA
  }
  n36 = Saturate(None=Multiply(A=Power(Base=@n31, Exp=@n32) {constExponent=3}, B=Multiply(A=@n33.Result, B=@n35.R)))
  n37 = NamedRerouteUsage()
  n38 = Add(A=Multiply(A=@n36, B=Multiply(A=@n37) {constB=0.7}), B=@n36)
  NamedRerouteDeclaration(None=Multiply(A=Saturate(None=@n38), B=ScalarParameter() {defaultValue=1.3, group=IgniteEgg, parameterName=IE_Ambient_GlitchIntensity})) {name=AmbientGlitchMask, nodeColor=<rgba 0.5234 0 1 1>}
  NamedRerouteDeclaration(None=@n1.EmissiveColor) {name=Emissive_In, nodeColor=<rgba 0.0312 0 1 1>}
  n39 = NamedRerouteUsage()
  n40 = Constant() {desc=Perlin pulse speed, r=0.4}
  n41 = NamedRerouteDeclaration(None=@n26.R) {name=Perlin, nodeColor=<rgba 1 0.8438 0 1>}
  n42 = NamedRerouteDeclaration(None=Multiply(
    A=MaterialFunctionCall(Speed=@n40) {materialFunction=&DebugTimeSine}.Result
    B=Power(Base=@n41)
   )) {name=PerlinPulse, nodeColor=<rgba 1 0.4453 0 1>}
  n43 = SmoothStep(Value=Multiply(A=Multiply(A=@n39, B=@n42)) {constB=2}) {constMin=0.5}
  NamedRerouteDeclaration(None=Multiply(A=Saturate(None=@n43)) {constB=0.5}) {name=WireframePulse, nodeColor=<rgba 0 1 0.8359 1>}
  NamedRerouteDeclaration(None=@n1.OpacityMask) {name=OpacityMask_In, nodeColor=<rgba 0.2891 1 0 1>}
  n44 = NamedRerouteUsage()
  n45 = OneMinus(None=Multiply(A=NamedRerouteUsage(), B=Constant() {r=0.4}))
  NamedRerouteDeclaration(None=MaterialFunctionCall(Alpha Threshold=Min(A=@n44, B=@n45)) {materialFunction=&DitherTemporalAA}.Result) {name=OpacityMask_OUT, nodeColor=<rgba 1 0 0.6094 1>}
  n46 = LocalPosition() {localOrigin=InstancePreSkinning}
  NamedRerouteDeclaration(None=ShaderStageSwitch(PixelShader=VertexInterpolator(VS=@n46.XYZ).PS, VertexShader=@n46.XYZ)) {name=PS-LocalPos, nodeColor=<rgba 0.1719 1 0 1>}
  n47 = Multiply(A=Time(), B=ScalarParameter() {defaultValue=0.75, group=IgniteEgg, parameterName=Debug_HitSpeed})
  n48 = Subtract(A=Time(), B=ScalarParameter() {group=IgniteEgg, parameterName=IE_GameTimeFromBP_Hit})
  n49 = ScalarParameter() {defaultValue=0.9, group=IgniteEgg, parameterName=IE_HitDuration}
  n50 = Saturate(None=StaticSwitchParameter(True=Frac(None=@n47), False=Divide(A=@n48, B=@n49)) {group=IgniteEgg, parameterName=DEBUG_Hit})
  n51 = Power(Base=OneMinus(None=@n50)) {constExponent=3}
  n52 = LocalPosition()
  n53 = VectorParameter() {group=IgniteEgg, parameterName=HitVector, defaultValue=<rgba 0 0 100 1>}
  n54 = LinearInterpolate(Alpha=@n50) {constA=10, constB=50}
  n55 = Multiply(A=Multiply(A=@n51, B=SphereMask(A=@n52.XYZ, B=@n53.RGB, Radius=@n54) {attenuationRadius=30, hardnessPercent=1})) {constB=5}
  n56 = NamedRerouteUsage()
  n57 = Saturate(None=Add(A=@n55, B=Subtract(A=@n55, B=@n56) {constB=0}))
  n58 = Multiply(
   A=Multiply(A=SmoothStep(Value=@n57) {constMax=0.5, constMin=0.4}, B=@n51)
   B=ScalarParameter() {defaultValue=2, group=IgniteEgg, parameterName=HitCenterIntencity}
  )
  n59 = Saturate(None=Divide(A=@n48, B=ScalarParameter() {defaultValue=1.1, group=IgniteEgg, parameterName=IE_HitDuration_2}))
  n60 = Power(Base=OneMinus(None=@n59)) {constExponent=3.8}
  n61 = SphereMask(A=@n52.XYZ, B=@n53.RGB, Radius=LinearInterpolate(Alpha=@n59) {constB=90}) {attenuationRadius=30, hardnessPercent=1}
  n62 = Multiply(A=@n61, B=@n60)
  n63 = NamedRerouteUsage()
  n64 = SphereMask(A=Subtract(A=@n61, B=Multiply(A=@n63) {constB=0.5}), B=Constant() {r=0.5}) {attenuationRadius=0.5, hardnessPercent=0}
  n65 = LinearInterpolate(Alpha=@n60) {constA=1, constB=0.3}
  n66 = Saturate(None=Sine(None=Multiply(A=@n64, B=@n65) {constB=2}))
  n67 = Multiply(
   A=Multiply(A=Multiply(A=@n62, B=@n66), B=ScalarParameter() {defaultValue=10, group=IgniteEgg, parameterName=IE_WavesIntensity})
   B=NamedRerouteUsage()
  ) {desc=Radial expanding wave}
  NamedRerouteDeclaration(None=Multiply(A=Add(A=@n58, B=@n67), B=ScalarParameter() {defaultValue=1, group=IgniteEgg, parameterName=HitIntensity})) {name=HitMask, nodeColor=<rgba 1 0 0.4922 1>}
  n68 = Add(A=Frac(None=Time()), B=NamedRerouteDeclaration(None=@n13.B) {name=Random, nodeColor=<rgba 0 0.2266 1 1>})
  n69 = NamedRerouteUsage()
  NamedRerouteDeclaration(None=LinearInterpolate(B=Frac(None=@n68), Alpha=OneMinus(None=@n69)) {constB=0}) {name=RandomGrayscale, nodeColor=<rgba 0 0.1562 1 1>}
  n70 = Subtract(A=Time(), B=ScalarParameter() {group=IgniteEgg, parameterName=IE_GameTimeFromBP})
  n71 = ScalarParameter() {defaultValue=1, group=IgniteEgg, parameterName=IE_HologramFX}
  n72 = LinearInterpolate(
   A=ScalarParameter() {defaultValue=1.2, group=IgniteEgg, parameterName=IE_HologramDuration_In}
   B=ScalarParameter() {defaultValue=1.75, group=IgniteEgg, parameterName=IE_HologramDuration_Out}
   Alpha=@n71
  )
  n73 = StaticSwitchParameter(True=ScalarParameter() {group=IgniteEgg, parameterName=Debug_Time, sliderMax=1}, False=Saturate(None=Divide(A=@n70, B=@n72))) {group=IgniteEgg, parameterName=DEBUG_Hologram}
  n74 = LinearInterpolate(A=@n73, B=OneMinus(None=@n73), Alpha=Clamp(None=@n71))
  n75 = Clamp(None=@n74, Max=ScalarParameter() {defaultValue=0.4, desc=0.5 blue, group=IgniteEgg, parameterName=IE_DBNO_HologramIntensity}) {desc="0, 0.35"}
  n76 = NamedRerouteDeclaration(None=ScalarParameter() {group=IgniteEgg, parameterName=IE_IsDBNO}) {name=IsDBNO, nodeColor=<rgba 1 0.8438 0 1>}
  n77 = Multiply(A=NamedRerouteDeclaration(None=LinearInterpolate(A=@n74, B=@n75, Alpha=@n76)) {name=HologramProgress, nodeColor=<rgba 1 0 0.75 1>}) {constB=2.5}
  n78 = NamedRerouteUsage()
  n79 = NamedRerouteUsage()
  n80 = LinearInterpolate(Alpha=@n76) {constA=0.8, constB=0.3, desc="Randomness 0.8, 0.3"}
  n81 = Max(A=Subtract(A=Subtract(A=@n77, B=@n78), B=Multiply(A=@n79, B=@n80) {constB=0.2}), B=NamedRerouteUsage())
  n82 = NamedRerouteUsage()
  NamedRerouteDeclaration(None=Saturate(None=Max(A=@n81, B=@n82))) {name=HologramMask, nodeColor=<rgba 0.875 1 0 1>}
  NamedRerouteUsage()
 ]}
